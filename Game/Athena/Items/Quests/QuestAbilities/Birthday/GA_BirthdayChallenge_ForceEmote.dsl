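; Birthday challenge quest ability that applies a dance gameplay effect to the owning player while they carry the cake-location tag, apparently to force an emote.
; On activation it runs the parent logic and waits for that tag to be added, then applies the effect and stores its handle.
; A nested wait for the tag's removal uses the stored handle to remove the effect.
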(event Ability|EventActivateAbility
  (|CallParentFunction)
  (Ability|Tasks|WaitGameplayTagAdd (Variables|GAVars|Setup|GetT_CakeLocation)
    (:Added
      (bind _returnvalue (GameplayEffects|ApplyGameplayEffectToSelf :self (Ability|GetAbilitySystemComponent (Variables|GAVars|Data|GetOwningPlayerPawn)) :GameplayEffectClass (Variables|GAVars|Setup|GetGE_BirthdayDance) :Level 0.0))
      (Variables|GAVars|Data|SetBirthday%20Gameplay%20Effect%20Handle _returnvalue)
      (Ability|Tasks|WaitGameplayTagRemove (Variables|GAVars|Setup|GetT_CakeLocation)
        (:Removed
          (GameplayEffects|RemoveActiveGameplayEffect (Ability|GetAbilitySystemComponent (Variables|GAVars|Data|GetOwningPlayerPawn)) (Variables|GAVars|Data|GetBirthday%20Gameplay%20Effect%20Handle) -1))))))

(event Ability|EventFailedToActivatePassiveAbility
  (params (ActorInfo :type Gameplay%20Ability%20Actor%20Info%20Structure)))
